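# Translucent, unlit, two-sided material for the Caretaker's bifurcate beam swipe effect, usable on Niagara mesh particles and ribbons.
# Color blends from time-of-day purple to particle color, then gets an HDR boost.
# Opacity combines panning noise, particle alpha, axis-aligned cylinder fresnel terms, end-cap, near-camera and depth fades (depth skipped on mobile).
# A scrolling gradient scales a vertex-normal offset, and UV-based width varies the cylinder alpha exponent.
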Material /Caretaker/FX/Swipe/M_Caretaker_BifucateBeam
graph {
 properties {
  bAllowNegativeEmissiveColor=true
  bUsedWithNiagaraMeshParticles=true
  bUsedWithNiagaraRibbons=true
  blendMode=BLEND_Translucent
  pixelDepthOffsetMode=PDOM_Legacy
  shadingModel=MSM_Unlit
  twoSided=true
  usedShadingModels=Unlit
 }
 nodes [
  n1 = VectorParameter() {parameterName=Low Color_NIght, defaultValue=<rgba 0.1946 0 0.449 1>}
  n2 = VectorParameter() {parameterName=Low Color, defaultValue=<rgba 0.1162 0 0.498 1>}
  n3 = ParticleColor()
  n4 = LinearInterpolate(
   A=MaterialFunctionCall(Color_Morning=@n1.RGB, Color_Day=@n2.RGB, Color_Evening=@n2.RGB, Color_Night=@n1.RGB) {materialFunction=&Color_Time_of_Day}.Result
   B=@n3.RGB
   Alpha=Power(Base=NamedRerouteUsage(), Exp=ScalarParameter() {defaultValue=8.4624, parameterName=Central Addition Exp})
  )
  MaterialFunctionCall(Emissive=MaterialFunctionCall(LDR Input=Saturate(None=@n4), HDR Tint=@n4, HDR Intensity Multiplier=ScalarParameter() {defaultValue=3, parameterName=HDR Intensity Day}) {materialFunction=&DeriveHDRfromLDR}.RemappedRange) {materialFunction=&MF_FX_EmissiveAdjustments}
  n5 = TextureCoordinate() {coordinateIndex=1}
  n6 = Power(
   Base=TextureSample(UVs=Panner(Coordinate=@n5) {speedX=-2, speedY=0.2}) {samplerType=SAMPLERTYPE_LinearGrayscale, texture=/Game/Effects/Fort_Effects/Textures/T_DarkAndLight_Noise}.G
   Exp=ScalarParameter() {defaultValue=2, parameterName=Texture Contrast}
  )
  n7 = Power(Base=NamedRerouteUsage(), Exp=ScalarParameter() {defaultValue=10, parameterName=Central Addition 2 Exp})
  n8 = Multiply(A=Multiply(A=@n3.A, B=Add(A=@n6, B=@n7)), B=NamedRerouteUsage())
  n9 = Sine(None=ComponentMask(None=TextureCoordinate()) {g=true}) {period=2}
  n10 = ScalarParameter() {defaultValue=2, parameterName=EndCapFadeExp}
  n11 = Saturate(None=Multiply(A=Power(Base=@n9, Exp=@n10), B=ScalarParameter() {defaultValue=10, parameterName=EndCapFadeIntensity}))
  n12 = ScalarParameter() {defaultValue=150, parameterName=Near Camera Fade Distance}
  n13 = ScalarParameter() {defaultValue=50, parameterName=Depth FadeDistance}
  Multiply(
   A=Multiply(
    A=Multiply(A=@n8, B=@n11)
    B=MaterialFunctionCall(Distance=@n12) {materialFunction=&Near_Camera_Fade}.Result
   )
   B=ShadingPathSwitch(Default=DepthFade(FadeDistance=@n13), Mobile=Constant() {r=1})
  )
  n14 = Panner() {speedY=-2}
  Multiply(
   A=VertexNormalWS()
   B=LinearInterpolate(Alpha=TextureSample(UVs=@n14) {texture=/Game/Global/Textures/Default/Masks_and_Gradients/gradient_stasis}.R) {constB=5}
  )
  n15 = VectorParameter() {parameterName=Cylinder Axis, defaultValue=<rgba 1 0 0 1>}
  NamedRerouteDeclaration(None=MaterialFunctionCall(Axis=@n15.RGB, Exponent=ScalarParameter() {defaultValue=25, parameterName="Cylinder Fresnel Exponent (S)"}) {materialFunction=&AxisAlignedFresnel}.Fresnel) {name=Cylinder Fresnel, nodeColor=<rgba 1 0 0.7031 1>}
  n16 = ScalarParameter() {defaultValue=2, parameterName="Cylinder Alpha Exponent (S)"}
  n17 = Multiply(A=@n16, B=ScalarParameter() {defaultValue=10, parameterName=INcrease Multyi})
  n18 = OneMinus(None=ComponentMask(None=TextureCoordinate()) {g=true})
  n19 = ScalarParameter() {defaultValue=10, parameterName=Width Contrast}
  n20 = Saturate(None=Multiply(A=Power(Base=@n18, Exp=@n19), B=ScalarParameter() {defaultValue=1, parameterName=Width Intensity}))
  NamedRerouteDeclaration(None=MaterialFunctionCall(Axis=@n15.RGB, Exponent=LinearInterpolate(A=@n16, B=@n17, Alpha=@n20)) {materialFunction=&AxisAlignedFresnel}.Fresnel) {name=Cylinder Alpha, nodeColor=<rgba 1 0.4922 0 1>}
 ]
}
