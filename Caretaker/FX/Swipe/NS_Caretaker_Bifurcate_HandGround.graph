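# Niagara effect for the Caretaker's Bifurcate swipe, by its name a hand-to-ground impact that throws up debris and smoke.
# Both sprite emitters combine bursts with a spawn rate, spawning in a low cylinder and pushing particles outward from a point.
# GPU Debris falls under heavy gravity and drag with lifetime-gated collision and random spin.
# Smoke brakes sharply, drifts upward, grows and fades, with camera offset, SubUV animation, a bright initial color and decaying rotation.
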NiagaraSystem /Caretaker/FX/Swipe/NS_Caretaker_Bifurcate_HandGround
dependencies {usedDataInterfaces [&NiagaraDataInterfaceCurve], usedDynamicInputs [], usedModules [&SystemState]}
emitters {
 Debris {
  simTarget=GPUComputeSim
  emitterUpdateScript [
   &EmitterState()
   &SpawnBurst_Instantaneous(Spawn Count=25, Use Spawn Probability=false, UseLoopCountLimit=false)
   &SpawnRate(SpawnRate=50, Use Spawn Probability=false)
  ]
  particleSpawnScript [
   &InitializeParticle(
    Color=<rgba 0 0 0 1>
    Lifetime Mode=::NewEnumerator1
    Mass Mode=::NewEnumerator3
    Sprite Size Max=<8 8>
    Sprite Size Min=<3 3>
    Sprite Size Mode=::NewEnumerator2
    UsePositionOffset=false
    Write Material Random=true
   )
   &CylinderLocation(
    Cylinder Axis:Vector3f=<0 0 1>
    Cylinder Height=10
    Cylinder Height Midpoint=0
    Cylinder Radius=50
    MaskSpawnGroup=false
    Offset=<0 0 -20>
    Offset Mode:ENiagara_OffsetMode=::NewEnumerator0
    Override Seed=false
    Rotation Coordinate Space:ENiagaraCoordinateSpace=::World
    Rotation Mode:ENiagara_RotationMode=::NewEnumerator1
    UseOwnerScale=true
   )
   &AddVelocityFromPoint(
    Module.Use Falloff Distance=false
    Origin Offset=<0 0 -25>
    Origin Offset Coordinate Space=::World
    Velocity Strength=&RandomRangeFloat(Minimum=500, Maximum=1000)
   )
   &SubUVAnimation(End Frame=3, SubUV Animation Mode:ENiagara_SubUVLookupMode=::NewEnumerator2)
  ]
  particleUpdateScript [
   &GravityForce(Gravity=<0 0 -2200>)
   &Drag(Drag=&RandomRangeFloat(Minimum=1, Maximum=3), Use Linear Drag=true, Use Rotational Drag=false)
   &Collision(
    Age Colliding Particles=true
    Collision Enabled=&SetBoolByFloatComparison(A=$Particles.Lifetime, B=0.12)
    Friction=1
    Friction During a Bounce=0.25
    GPU Collision Type=::NewEnumerator2
    Module.Control Rest State Via Surface Normal=false
    Module.EnableMaxCollisionCount=false
    Module.Randomize Collision Normal Vector=true
    Randomize Collision Normal=0.25
    Restitution=0.75
   )
   &SolveForcesAndVelocity(Clamp Velocity=false, Limit Acceleration=false)
   &ParticleState()
   &SpriteRotationRate(Rotation Rate=&Lerp_Float(A=-70, B=70, Alpha=$Particles.Initial.MaterialRandom))
   &ScaleSpriteSize(Uniform Curve Sprite Scale=DataInterface() {propertyValues {Curve {keys [
        {interpMode=RCIM_Cubic, tangentMode=RCTM_User, time=0, value=0}
        {interpMode=RCIM_Cubic, tangentMode=RCTM_User, time=0.2, value=1}
        {arriveTangentWeight=0.3202, interpMode=RCIM_Cubic, leaveTangentWeight=0.1067, tangentMode=RCTM_User, time=0.8, value=1}
        {tangentMode=RCTM_User, time=1, value=0}
       ]}}})
  ]
  rendererClasses [&NiagaraSpriteRendererProperties]
  renderers [{rendererClass=&NiagaraSpriteRendererProperties}]
 }
 Smoke {
  emitterUpdateScript [
   &EmitterState()
   &SpawnRate(SpawnRate=100, Use Spawn Probability=false)
   &SpawnBurst_Instantaneous(Spawn Count=50, Use Spawn Probability=false, UseLoopCountLimit=false) {module=SpawnBurst_Instantaneous001}
   &SpawnBurst_Instantaneous(Spawn Count=25, Spawn Time=0.04, Use Spawn Probability=false, UseLoopCountLimit=false)
  ]
  particleSpawnScript [
   &InitializeParticle(
    Color Maximum=<rgba 0.2344 0.2074 0.2097 0.75>
    Color Minimum=<rgba 0.125 0.1036 0.117 0.5>
    Color Mode=::NewEnumerator2
    Lifetime Max=1.5
    Lifetime Min=0.8
    Lifetime Mode=::NewEnumerator1
    Sprite Rotation Mode=::NewEnumerator2
    Sprite Size Mode=::NewEnumerator1
    Uniform Sprite Size Max=350
    Uniform Sprite Size Min=250
    UsePositionOffset=false
    Write Material Random=true
   )
   &CylinderLocation(
    Cylinder Axis:Vector3f=<0 0 1>
    Cylinder Height=10
    Cylinder Height Midpoint=0
    Cylinder Radius=50
    MaskSpawnGroup=false
    Offset=<0 0 -20>
    Offset Mode:ENiagara_OffsetMode=::NewEnumerator0
    Override Seed=false
    Rotation Coordinate Space:ENiagaraCoordinateSpace=::World
    Rotation Mode:ENiagara_RotationMode=::NewEnumerator1
    UseOwnerScale=true
   )
   &AddVelocityFromPoint(
    Module.Use Falloff Distance=false
    Origin Offset Coordinate Space=::Simulation
    Velocity Strength=&RandomRangeFloat(Minimum=600, Maximum=1750)
   )
  ]
  particleUpdateScript [
   &ParticleState()
   &Drag(Drag=&RandomRangeFloat(Minimum=5, Maximum=7), Use Linear Drag=true, Use Rotational Drag=false)
   &AccelerationForce(Acceleration=<0 0 40>)
   &SolveForcesAndVelocity(Clamp Velocity=false, Limit Acceleration=false)
   &CameraOffset(Camera Offset Amount=&Multiply_Float(
     A=0.5
     B=/Niagara/DynamicInputs/Math/CalculateParticleRadius(
      Radius Calculation Type:ENiagaraCollisionRadiusOptions=::NewEnumerator0
      Particle Radius:NiagaraFloat=1
      Mesh Dimensions:Vector3f=<1 1 1>
      Method for Calculating Particle Radius:ENiagaraCalculateRadiusOptions=::NewEnumerator2
     )
    ))
   &ScaleSpriteSize(Uniform Curve Sprite Scale=DataInterface() {propertyValues {Curve {keys [
        {arriveTangent=2.7535, interpMode=RCIM_Cubic, leaveTangent=2.7535, tangentMode=RCTM_User, time=0, value=0.25}
        {arriveTangent=0.5339, interpMode=RCIM_Cubic, leaveTangent=-0.3011, tangentMode=RCTM_User, time=1, value=1.25}
       ]}}})
   &SubUVAnimation(Start Frame=20)
   &ScaleColor(
    Scale Alpha=&FloatFromCurve(FloatCurve=DataInterface() {propertyValues {Curve {keys [
         {arriveTangent=8.5887, interpMode=RCIM_Cubic, leaveTangent=8.5887, tangentMode=RCTM_User, time=0, value=0}
         {arriveTangent=-1.4353, interpMode=RCIM_Cubic, leaveTangent=-1.4353, tangentMode=RCTM_User, time=0.3, value=1}
         {interpMode=RCIM_Cubic, leaveTangent=-3.4295, time=1, value=0}
        ]}}})
    Scale RGB=&VectorFromFloat(Value=&FloatFromCurve(
      FloatCurve=DataInterface() {propertyValues {Curve {keys [{arriveTangent=-11.0936, interpMode=RCIM_Cubic, leaveTangent=-11.0936, tangentMode=RCTM_User, time=0, value=3}, <at 0.4 1>]}}}
      CurveIndex=$Particles.Age
     ))
    ScaleA=true
    ScaleRGB=true
   )
   &SpriteRotationRate(Rotation Rate=&MakeCustomFloatFromBool(
     Bool=&RandomBool()
     True Float=&Multiply_Float(
      A=&Lerp_Float(A=-200, B=200, Alpha=$Particles.MaterialRandom)
      B=&FloatFromCurve(FloatCurve=DataInterface() {propertyValues {Curve {keys [
           {arriveTangent=-4.4572, interpMode=RCIM_Cubic, time=0, value=1}
           {arriveTangent=-2.0073, interpMode=RCIM_Cubic, leaveTangent=-2.0073, tangentMode=RCTM_User, time=0.25, value=0.25}
           <at 0.5 0.01>
          ]}}})
     )
     False Float=0
    ))
  ]
  rendererClasses [&NiagaraSpriteRendererProperties]
  renderers [{rendererClass=&NiagaraSpriteRendererProperties}]
 }
}
summary {
 systemName=NS_Caretaker_Bifurcate_HandGround
 emitters [
  {emitterName=Smoke, rendererClasses [&NiagaraSpriteRendererProperties]}
  {emitterName=Debris, simTarget=GPUComputeSim, rendererClasses [&NiagaraSpriteRendererProperties]}
 ]
 userVariables []
}
